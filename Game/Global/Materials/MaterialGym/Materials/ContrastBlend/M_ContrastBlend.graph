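# M_ContrastBlend is an unlit MaterialGym material graph that prototypes a noise-driven contrast blend.
# Inverted, scrolling cell noise scaled by a Contrast parameter sets how far an oppositely scrolling blurred noise is remapped around a Bias pivot, then saturated.
# A separate SmoothStep builds a band mask centered along U.
# The graph shows no connection to material outputs and retains stray Add nodes.
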Material /Game/Global/Materials/MaterialGym/Materials/ContrastBlend/M_ContrastBlend
graph {
 properties {pixelDepthOffsetMode=PDOM_Legacy, shadingModel=MSM_Unlit, usedShadingModels=Unlit}
 nodes [
  n1 = TextureCoordinate()
  n2 = Time()
  n3 = Constant2Vector() {g=0.2}
  n4 = TextureSample(UVs=Add(A=Multiply(A=@n1) {constB=2}, B=Multiply(A=@n2, B=@n3))) {texture=/Game/Global/Textures/Default/Noise/T_CellNoise_01}
  Add(A=@n4.R) {constB=0}
  Add()
  NamedRerouteDeclaration(None=ScalarParameter() {defaultValue=0.5, parameterName=Bias}) {name=Bias, nodeColor=<rgba 0.2656 0 1 1>}
  n5 = Add(A=TextureCoordinate(), B=Multiply(A=Time(), B=Constant2Vector() {g=-0.1}))
  n6 = OneMinus(None=@n4.R)
  n7 = ScalarParameter() {defaultValue=0.2, parameterName=Contrast}
  n8 = Clamp(None=Saturate(None=Multiply(A=@n6, B=@n7) {constB=0.5})) {maxDefault=0.5, minDefault=-0.5}
  n9 = Subtract(
   A=Subtract(
    A=TextureSample(UVs=@n5) {samplerType=SAMPLERTYPE_Grayscale, texture=&LowResBlurredNoise}.R
    B=NamedRerouteUsage()
   )
   B=@n8
  )
  n10 = Subtract(A=Subtract(B=@n8), B=@n8)
  Saturate(None=Add(A=NamedRerouteUsage(), B=Divide(A=@n9, B=@n10)))
  n11 = ConstantBiasScale(None=ComponentMask(None=TextureCoordinate()) {r=true}) {bias=-0.5, scale=2}
  SmoothStep(Value=OneMinus(None=Abs(None=@n11))) {constMin=0.5}
 ]
}
