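# Unlit Material Gym test material that builds a contrast-blend mask, with no material output connections shown.
# A scrolling, tiled cell-noise texture is inverted, scaled by Contrast and clamped to form the blend offset.
# Reroute inputs, with Fresnel Gradient and Bias declared, are remapped around that offset by subtract, divide and saturate, breaking up the transition with noise.
# An alternative UV or scrolling-noise gradient switch sits unconnected.
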Material /Game/Global/Materials/MaterialGym/Materials/ContrastBlend/M_ContrastBlend3
graph {
 properties {pixelDepthOffsetMode=PDOM_Legacy, shadingModel=MSM_Unlit, usedShadingModels=Unlit}
 nodes [
  NamedRerouteDeclaration(None=ScalarParameter() {defaultValue=0.5, parameterName=Bias}) {name=Bias, nodeColor=<rgba 0.2656 0 1 1>}
  NamedRerouteDeclaration(None=Fresnel(ExponentIn=ScalarParameter() {defaultValue=1, parameterName=ExponentIn}) {baseReflectFraction=0}) {name=Gradient, nodeColor=<rgba 1 0.8672 0 1>}
  n1 = TextureCoordinate()
  n2 = ComponentMask(None=@n1) {g=true}
  OneMinus(None=@n2)
  n3 = Multiply(A=TextureCoordinate(), B=ScalarParameter() {defaultValue=1, parameterName=BlendTextureTiling}) {constB=2}
  n4 = Multiply(A=Time(), B=ComponentMask(None=VectorParameter() {parameterName=BlendTextureScrollSpeed, defaultValue=<rgba 0 0 0 1>}.RGB) {g=true, r=true})
  n5 = OneMinus(None=TextureSampleParameter2D(UVs=Add(A=@n3, B=@n4)) {parameterName=BlendTexture, texture=/Game/Global/Textures/Default/Noise/T_CellNoise_01}.R)
  n6 = ScalarParameter() {defaultValue=0.25, parameterName=Contrast}
  n7 = Clamp(None=Saturate(None=Multiply(A=@n5, B=@n6) {constB=0.5})) {maxDefault=0.5, minDefault=-0.5}
  n8 = Subtract(A=Subtract(A=NamedRerouteUsage(), B=NamedRerouteUsage()), B=@n7)
  n9 = Subtract(A=Subtract(B=@n7), B=@n7)
  Saturate(None=Add(A=NamedRerouteUsage(), B=Divide(A=@n8, B=@n9)))
  n10 = Time()
  n11 = ComponentMask(None=VectorParameter() {parameterName=TextureGradientSpeed, defaultValue=<rgba 0 0 0 1>}.RGB) {g=true, r=true}
  n12 = TextureSampleParameter2D(UVs=Add(A=@n1, B=Multiply(A=@n10, B=@n11))) {parameterName=GradientTexture, samplerType=SAMPLERTYPE_Grayscale, texture=&LowResBlurredNoise}
  StaticSwitchParameter(True=@n12.R, False=@n2) {parameterName=UseTextureGradient}
 ]
}
